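; Editor utility Blueprint that packs seven directional cloud textures (front/back left, center, right, plus top) into two BC7 textures and bakes a matching static mesh.
; Construction previews the bake: a jump-flood pass feeds defringed packing materials, composited with a separable alpha blur.
; BakeCloud saves both composites to textures.
; Its mesh step displaces a grid by dilated blurred alpha, plane-cuts, flattens, simplifies and remeshes it via Geometry Script into the cutout.
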(event Custom|BakeCloudTextures
  (Utilities|FlowControl|Sequence
    (:then_0
      (for _array_element (Utilities|Array|MakeArray (Variables|Textures|GetFrontLeft) (Variables|Textures|GetFront) (Variables|Textures|GetFrontRight) (Variables|Textures|GetBackLeft) (Variables|Textures|GetBack) (Variables|Textures|GetBackRight) (Variables|Textures|GetTop))
        (Variables|/Script/Engine.Texture|SetCompressionSettings "TC_EditorIcon" _array_element)))
    (:then_1
      (bind _returnvalue (Rendering|Material|CreateDynamicMaterialInstance "/Game/NeverCook/EditorUtilities/BakedCloudsPacker/M_BakedCloudsPacker_JFA_Init.M_BakedCloudsPacker_JFA_Init" "None" "None" :!class /Script/Engine.KismetMaterialLibrary))
      (Variables|Default|SetMaterial_JFA_Init _returnvalue)
      (Rendering|Material|SetTextureParameterValue (Variables|Default|GetMaterial_JFA_Init) "AlphaTexture" (Variables|Textures|GetFrontLeft))
      (Rendering|ClearRenderTarget2D (Variables|Default|GetRT_JFA))
      (bind _returnvalue_2 (Rendering|CreateRenderTarget2D 1024 1024 "RTF_RG32f"))
      (bind _output_get (Variables|Default|SetRT_JFA _returnvalue_2))
      (Rendering|DrawMaterialtoRenderTarget _output_get (Variables|Default|GetMaterial_JFA_Init))
      (bind _returnvalue_3 (Rendering|Material|CreateDynamicMaterialInstance "/Game/NeverCook/EditorUtilities/BakedCloudsPacker/M_BakedCloudsPacker_JFA_Pass.M_BakedCloudsPacker_JFA_Pass" "None" "None" :!class /Script/Engine.KismetMaterialLibrary))
      (Variables|Default|SetMaterial_JFA_Pass _returnvalue_3)
      (Rendering|Material|SetTextureParameterValue (Variables|Default|GetMaterial_JFA_Pass) "RT" (Variables|Default|GetRT_JFA))
      (for _index (range 1 (+ (Math|Float|Truncate (Math|Float|Log 512.0 2.0)) 1))
        (Rendering|Material|SetScalarParameterValue (Variables|Default|GetMaterial_JFA_Pass) "Step" (/ 1.0 (Math|Float|Power 2.0 (Math|Conversions|ToFloat(Integer) _index))) :!class /Script/Engine.MaterialInstanceDynamic)
        (bind (_canvas _size _context) (Rendering|BeginDrawCanvastoRenderTarget (Variables|Default|GetRT_JFA)))
        (Canvas|DrawMaterial :self _canvas :RenderMaterial (Variables|Default|GetMaterial_JFA_Pass) :ScreenSize _size :CoordinateSize "(X=1.000000,Y=1.000000)" :Rotation 0.0 :PivotPoint "(X=0.500000,Y=0.500000)")
        (Rendering|EndDrawCanvastoRenderTarget _context)))
    (:then_2
      (bind _returnvalue_4 (Rendering|Material|CreateDynamicMaterialInstance "/Game/NeverCook/EditorUtilities/BakedCloudsPacker/M_BakedCloudsPacker_Pack_A.M_BakedCloudsPacker_Pack_A" "None" "None" :!class /Script/Engine.KismetMaterialLibrary))
      (Variables|Default|SetMaterial_Pack_A _returnvalue_4)
      (Rendering|Material|SetTextureParameterValue (Variables|Default|GetMaterial_Pack_A) "FrontLeft" (Variables|Textures|GetFrontLeft))
      (Rendering|Material|SetScalarParameterValue (Variables|Default|GetMaterial_Pack_A) "FrontLeftShadows" (Variables|Textures|GetFrontLeftShadows) :!class /Script/Engine.MaterialInstanceDynamic)
      (Rendering|Material|SetScalarParameterValue (Variables|Default|GetMaterial_Pack_A) "FrontLeftHighlights" (Variables|Textures|GetFrontLeftHighlights) :!class /Script/Engine.MaterialInstanceDynamic)
      (Rendering|Material|SetTextureParameterValue (Variables|Default|GetMaterial_Pack_A) "Front" (Variables|Textures|GetFront))
      (Rendering|Material|SetScalarParameterValue (Variables|Default|GetMaterial_Pack_A) "FrontShadows" (Variables|Textures|GetFrontShadows) :!class /Script/Engine.MaterialInstanceDynamic)
      (Rendering|Material|SetScalarParameterValue (Variables|Default|GetMaterial_Pack_A) "FrontHighlights" (Variables|Textures|GetFrontHighlights) :!class /Script/Engine.MaterialInstanceDynamic)
      (Rendering|Material|SetTextureParameterValue (Variables|Default|GetMaterial_Pack_A) "FrontRight" (Variables|Textures|GetFrontRight))
      (Rendering|Material|SetScalarParameterValue (Variables|Default|GetMaterial_Pack_A) "FrontRightShadows" (Variables|Textures|GetFrontRightShadows) :!class /Script/Engine.MaterialInstanceDynamic)
      (Rendering|Material|SetScalarParameterValue (Variables|Default|GetMaterial_Pack_A) "FrontRightHighlights" (Variables|Textures|GetFrontRightHighlights) :!class /Script/Engine.MaterialInstanceDynamic)
      (Rendering|Material|SetTextureParameterValue (Variables|Default|GetMaterial_Pack_A) "JFA" (Variables|Default|GetRT_JFA))
      (Rendering|Material|SetScalarParameterValue (Variables|Default|GetMaterial_Pack_A) "DefringeRadius" (Variables|Adjustments|GetDefringeRadius) :!class /Script/Engine.MaterialInstanceDynamic)
      (bind _returnvalue_5 (Rendering|CreateRenderTarget2D 1024 1024 "RTF_RGBA8" "(R=0.000000,G=0.000000,B=0.000000,A=1.000000)" true))
      (bind _output_get_2 (Variables|Default|SetRT_Pack_A _returnvalue_5))
      (Rendering|DrawMaterialtoRenderTarget _output_get_2 (Variables|Default|GetMaterial_Pack_A)))
    (:then_3
      (bind _returnvalue_6 (Rendering|Material|CreateDynamicMaterialInstance "/Game/NeverCook/EditorUtilities/BakedCloudsPacker/M_BakedCloudsPacker_Pack_B.M_BakedCloudsPacker_Pack_B" "None" "None" :!class /Script/Engine.KismetMaterialLibrary))
      (Variables|Default|SetMaterial_Pack_B _returnvalue_6)
      (Rendering|Material|SetTextureParameterValue (Variables|Default|GetMaterial_Pack_B) "BackLeft" (Variables|Textures|GetBackLeft))
      (Rendering|Material|SetScalarParameterValue (Variables|Default|GetMaterial_Pack_B) "BackLeftShadows" (Variables|Textures|GetBackLeftShadows) :!class /Script/Engine.MaterialInstanceDynamic)
      (Rendering|Material|SetScalarParameterValue (Variables|Default|GetMaterial_Pack_B) "BackLeftHighlights" (Variables|Textures|GetBackLeftHighlights) :!class /Script/Engine.MaterialInstanceDynamic)
      (Rendering|Material|SetTextureParameterValue (Variables|Default|GetMaterial_Pack_B) "Back" (Variables|Textures|GetBack))
      (Rendering|Material|SetScalarParameterValue (Variables|Default|GetMaterial_Pack_B) "BackShadows" (Variables|Textures|GetBackShadows) :!class /Script/Engine.MaterialInstanceDynamic)
      (Rendering|Material|SetScalarParameterValue (Variables|Default|GetMaterial_Pack_B) "BackHighlights" (Variables|Textures|GetBackHighlights) :!class /Script/Engine.MaterialInstanceDynamic)
      (Rendering|Material|SetTextureParameterValue (Variables|Default|GetMaterial_Pack_B) "BackRight" (Variables|Textures|GetBackRight))
      (Rendering|Material|SetScalarParameterValue (Variables|Default|GetMaterial_Pack_B) "BackRightShadows" (Variables|Textures|GetBackRightShadows) :!class /Script/Engine.MaterialInstanceDynamic)
      (Rendering|Material|SetScalarParameterValue (Variables|Default|GetMaterial_Pack_B) "BackRightHighlights" (Variables|Textures|GetBackRightHighlights) :!class /Script/Engine.MaterialInstanceDynamic)
      (Rendering|Material|SetTextureParameterValue (Variables|Default|GetMaterial_Pack_B) "Top" (Variables|Textures|GetTop))
      (Rendering|Material|SetScalarParameterValue (Variables|Default|GetMaterial_Pack_B) "TopShadows" (Variables|Textures|GetTopShadows) :!class /Script/Engine.MaterialInstanceDynamic)
      (Rendering|Material|SetScalarParameterValue (Variables|Default|GetMaterial_Pack_B) "TopHighlights" (Variables|Textures|GetTopHighlights) :!class /Script/Engine.MaterialInstanceDynamic)
      (Rendering|Material|SetTextureParameterValue (Variables|Default|GetMaterial_Pack_B) "JFA" (Variables|Default|GetRT_JFA))
      (Rendering|Material|SetScalarParameterValue (Variables|Default|GetMaterial_Pack_B) "DefringeRadius" (Variables|Adjustments|GetDefringeRadius) :!class /Script/Engine.MaterialInstanceDynamic)
      (bind _returnvalue_7 (Rendering|CreateRenderTarget2D 1024 1024 "RTF_RGBA8" "(R=0.000000,G=0.000000,B=0.000000,A=1.000000)" true))
      (bind _output_get_3 (Variables|Default|SetRT_Pack_B _returnvalue_7))
      (Rendering|DrawMaterialtoRenderTarget _output_get_3 (Variables|Default|GetMaterial_Pack_B)))
    (:then_4
      (bind _returnvalue_8 (Rendering|Material|CreateDynamicMaterialInstance "/Game/NeverCook/EditorUtilities/BakedCloudsPacker/M_BakedCloudsPacker_BlurH.M_BakedCloudsPacker_BlurH" "None" "None" :!class /Script/Engine.KismetMaterialLibrary))
      (Variables|Default|SetAlphaBlurH _returnvalue_8)
      (Rendering|Material|SetTextureParameterValue (Variables|Default|GetAlphaBlurH) "AlphaTexture" (Variables|Textures|GetFrontLeft))
      (Rendering|Material|SetScalarParameterValue (Variables|Default|GetAlphaBlurH) "Radius" (Variables|Adjustments|GetBlurRadius) :!class /Script/Engine.MaterialInstanceDynamic)
      (bind _returnvalue_9 (Rendering|CreateRenderTarget2D 1024 1024))
      (bind _output_get_4 (Variables|Default|SetRT_Blur_H _returnvalue_9))
      (Rendering|DrawMaterialtoRenderTarget _output_get_4 (Variables|Default|GetAlphaBlurH))
      (bind _returnvalue_10 (Rendering|Material|CreateDynamicMaterialInstance "/Game/NeverCook/EditorUtilities/BakedCloudsPacker/M_BakedCloudsPacker_BlurV.M_BakedCloudsPacker_BlurV" "None" "None" :!class /Script/Engine.KismetMaterialLibrary))
      (Variables|Default|SetAlphaBlurV _returnvalue_10)
      (Rendering|Material|SetTextureParameterValue (Variables|Default|GetAlphaBlurV) "AlphaTexture" (Variables|Default|GetRT_Blur_H))
      (Rendering|Material|SetScalarParameterValue (Variables|Default|GetAlphaBlurV) "Radius" (Variables|Adjustments|GetBlurRadius) :!class /Script/Engine.MaterialInstanceDynamic)
      (bind _returnvalue_11 (Rendering|CreateRenderTarget2D 1024 1024 "RTF_R16f"))
      (bind _output_get_5 (Variables|Default|SetRT_Blur _returnvalue_11))
      (Rendering|DrawMaterialtoRenderTarget _output_get_5 (Variables|Default|GetAlphaBlurV)))
    (:then_5
      (bind _returnvalue_12 (Rendering|Material|CreateDynamicMaterialInstance "/Game/NeverCook/EditorUtilities/BakedCloudsPacker/M_BakedCloudsPacker_Comp.M_BakedCloudsPacker_Comp" "None" "None" :!class /Script/Engine.KismetMaterialLibrary))
      (Variables|Default|SetMaterial_Comp _returnvalue_12)
      (Rendering|Material|SetTextureParameterValue (Variables|Default|GetMaterial_Comp) "Pack_A" (Variables|Default|GetRT_Pack_A))
      (Rendering|Material|SetTextureParameterValue (Variables|Default|GetMaterial_Comp) "Pack_B" (Variables|Default|GetRT_Pack_B))
      (Rendering|Material|SetTextureParameterValue (Variables|Default|GetMaterial_Comp) "Blur" (Variables|Default|GetRT_Blur))
      (Rendering|Material|SetTextureParameterValue (Variables|Default|GetMaterial_Comp) "JFA" (Variables|Default|GetRT_JFA))
      (Rendering|Material|SetScalarParameterValue (Variables|Default|GetMaterial_Comp) "BlurEnd" (Variables|Adjustments|GetBlurEnd) :!class /Script/Engine.MaterialInstanceDynamic)
      (Rendering|Material|SetScalarParameterValue (Variables|Default|GetMaterial_Comp) "BlurStart" (Variables|Adjustments|GetBlurStart) :!class /Script/Engine.MaterialInstanceDynamic)
      (Rendering|Material|SetScalarParameterValue (Variables|Default|GetMaterial_Comp) "ScaleX" (Variables|Adjustments|GetUV%20ScaleX) :!class /Script/Engine.MaterialInstanceDynamic)
      (Rendering|Material|SetScalarParameterValue (Variables|Default|GetMaterial_Comp) "ScaleY" (Variables|Adjustments|GetUV%20ScaleY) :!class /Script/Engine.MaterialInstanceDynamic)
      (Rendering|Material|SetScalarParameterValue (Variables|Default|GetMaterial_Comp) "OffsetX" (Variables|Adjustments|GetUV%20Offset%20X) :!class /Script/Engine.MaterialInstanceDynamic)
      (Rendering|Material|SetScalarParameterValue (Variables|Default|GetMaterial_Comp) "OffsetY" (Variables|Adjustments|GetUV%20Offset%20Y) :!class /Script/Engine.MaterialInstanceDynamic)
      (bind _returnvalue_13 (Rendering|CreateRenderTarget2D (Utilities|Select :Index (Variables|Output|GetBakedTextureResolutionX) :NewEnumerator0 256 :NewEnumerator1 512 :NewEnumerator10 1024) (Utilities|Select :Index (Variables|Output|GetBakedTextureResolutionY) :NewEnumerator0 256 :NewEnumerator1 512 :NewEnumerator10 1024) "RTF_RGBA8"))
      (bind _output_get_6 (Variables|Default|SetRT_Comp _returnvalue_13))
      (Rendering|DrawMaterialtoRenderTarget _output_get_6 (Variables|Default|GetMaterial_Comp)))
    (:then_6
      (Default|DrawPreview))))

(event Custom|BakeCloudMeshes
  (bind _returnvalue (Rendering|Material|CreateDynamicMaterialInstance "/Game/NeverCook/EditorUtilities/BakedCloudsPacker/M_BakedCloudsPacker_DisplaceDilation.M_BakedCloudsPacker_DisplaceDilation" "None" "None" :!class /Script/Engine.KismetMaterialLibrary))
  (Variables|Default|SetDisplaceMaterialDynamic _returnvalue)
  (Rendering|Material|SetScalarParameterValue (Variables|Default|GetDisplaceMaterialDynamic) "DilationAmount" (Variables|Output|GetStaticMeshBorderDilation) :!class /Script/Engine.MaterialInstanceDynamic)
  (Rendering|Material|SetScalarParameterValue (Variables|Default|GetDisplaceMaterialDynamic) "ScaleX" (Variables|Adjustments|GetUV%20ScaleX) :!class /Script/Engine.MaterialInstanceDynamic)
  (Rendering|Material|SetScalarParameterValue (Variables|Default|GetDisplaceMaterialDynamic) "ScaleY" (Variables|Adjustments|GetUV%20ScaleY) :!class /Script/Engine.MaterialInstanceDynamic)
  (Rendering|Material|SetScalarParameterValue (Variables|Default|GetDisplaceMaterialDynamic) "OffsetX" (Variables|Adjustments|GetUV%20Offset%20X) :!class /Script/Engine.MaterialInstanceDynamic)
  (Rendering|Material|SetScalarParameterValue (Variables|Default|GetDisplaceMaterialDynamic) "OffsetY" (Variables|Adjustments|GetUV%20Offset%20Y) :!class /Script/Engine.MaterialInstanceDynamic)
  (Rendering|Material|SetTextureParameterValue (Variables|Default|GetDisplaceMaterialDynamic) "BlurredAlpha" (Variables|Default|GetRT_Blur))
  (bind _returnvalue_2 (Rendering|CreateRenderTarget2D 1024 1024))
  (bind _output_get (Variables|Default|SetDisplaceTexture _returnvalue_2))
  (Rendering|DrawMaterialtoRenderTarget _output_get (Variables|Default|GetDisplaceMaterialDynamic))
  (EditorScripting|Rendering|ConvertRenderTargettoTexture2DEditorOnly (Variables|Default|GetDisplaceTexture) "/Game/NeverCook/EditorUtilities/BakedCloudsPacker/T_BakedClouds_DisplaceTexture.T_BakedClouds_DisplaceTexture")
  (bind _returnvalue_3 (AddComponent|NODE_AddDynamicMeshComponent-2))
  (bind _output_get_2 (Variables|Default|SetDynamicMeshComponent _returnvalue_3))
  (bind _returnvalue_4 (DynamicMeshComponent|GetDynamicMesh _output_get_2))
  (bind _returnvalue_5 (GeometryScript|Primitives|AppendRectangleXY :TargetMesh _returnvalue_4 :PrimitiveOptions (struct :PolygroupMode "SingleGroup" :bFlipOrientation false :UVMode "Uniform" :MaterialID 0) :DimensionX (Variables|Adjustments|GetMeshSizeX) :DimensionY (Variables|Adjustments|GetMeshSizeY) :StepsWidth 50 :StepsHeight 50))
  (bind _returnvalue_6 (GeometryScript|UVs|ScaleMeshUVs _returnvalue_5 0 (Utilities|Select (Math|Vector2D|MakeVector2D (/ (Variables|Adjustments|GetMeshSizeY) (Variables|Adjustments|GetMeshSizeX)) 1.0) (Math|Vector2D|MakeVector2D 1.0 (/ (Variables|Adjustments|GetMeshSizeX) (Variables|Adjustments|GetMeshSizeY))) (>= (Variables|Adjustments|GetMeshSizeX) (Variables|Adjustments|GetMeshSizeY))) "(X=0.000000,Y=0.000000)"))
  (bind _returnvalue_7 (GeometryScript|Deformations|ApplyDisplacefromTextureMap :TargetMesh _returnvalue_6 :Texture "/Game/NeverCook/EditorUtilities/BakedCloudsPacker/T_BakedClouds_DisplaceTexture.T_BakedClouds_DisplaceTexture" :Options (Utilities|Struct|MakeGeometryScriptDisplacefromTextureOptions (/ (Math|Float|Max(Float) (Variables|Adjustments|GetMeshSizeX) (Variables|Adjustments|GetMeshSizeY)) 1e+02) "(X=1.000000,Y=1.000000)" "(X=0.000000,Y=0.000000)" 0.0) :TessellationOptions (Utilities|Struct|MakeGeometryScriptAdaptiveTessellationOptions true 0.5 "Relative" 0.01 false) :UVLayer 0))
  (bind _returnvalue_8 (GeometryScript|Booleans|ApplyMeshPlaneCut _returnvalue_7 (struct :Location "0.000000,0.000000,0.100000" :Rotation "0, 0, 0" :Scale "1.000000,1.000000,1.000000") (Utilities|Struct|MakeGeometryScriptMeshPlaneCutOptions false -1 false true)))
  (bind _returnvalue_9 (GeometryScript|Transforms|ScaleMesh _returnvalue_8 "1.000000,1.000000,0.000000"))
  (Utilities|FlowControl|Delay 1.0)
  (bind _returnvalue_10 (GeometryScript|Simplification|ApplySimplifytoEdgeLength _returnvalue_9 (* (/ (Math|Float|Max(Float) (Variables|Adjustments|GetMeshSizeX) (Variables|Adjustments|GetMeshSizeY)) 1e+02) (Variables|Output|GetStaticMeshSimplifyEdge)) (struct :Method "VolumePreserving" :bAllowSeamCollapse true :bAllowSeamSmoothing true :bAllowSeamSplits true :bPreserveVertexPositions false :bRetainQuadricMemory true :EdgeLengthWeightMap "(Handle=(WeightMapAttributeLayerIndex=-1),RelativeDensity=0.000000)" :GeometricToleranceWeightMap "(Handle=(WeightMapAttributeLayerIndex=-1),RelativeDensity=0.000000)" :QuadricErrorWeightMap "(Handle=(WeightMapAttributeLayerIndex=-1),RelativeDensity=0.000000)" :RegularizeWeight 1e-06 :bAutoCompact true :QuadricVariant "PlaneQuadric" :NormalAttributeWeight 16.0 :TangentAttributeWeight 0.1 :ColorAttributeWeight 0.1 :TexCoordAttributeWeight 0.5 :ScaleCorrection 1.0)))
  (Utilities|FlowControl|Delay 0.5)
  (bind _returnvalue_11 (GeometryScript|Simplification|ApplyUniformRemesh _returnvalue_10 (Utilities|Struct|MakeGeometryScriptRemeshOptions false true "Mixed" 0.25 "Free" "Free" "Free" true true true true true true) (Utilities|Struct|MakeGeometryScriptUniformRemeshOptions "TriangleCount" (Variables|Output|GetStaticMeshSimplifyTesselation))))
  (bind _returnvalue_12 (GeometryScript|Normals|RecomputeNormals _returnvalue_11))
  (bind _returnvalue_13 (GeometryScript|Transforms|RotateMesh _returnvalue_12 "0.000000,0.000000,90.000000"))
  (bind _returnvalue_14 (GeometryScript|StaticMesh|CopyMeshtoStaticMesh _returnvalue_13 (Variables|Output|GetStaticMesh))
    (:Failure
      _)
    (:Success
      _))
  (DynamicMesh|Reset _returnvalue_14)
  (Components|DestroyComponent (Variables|Default|GetDynamicMeshComponent)))

(fn BakeCloud
  (bind _bakedtexture_a (Variables|Output|GetBakedTexture_A))
  (bind _bakedtexture_b (Variables|Output|GetBakedTexture_B))
  (CallFunction|BakeCloudTextures)
  (CallFunction|BakeCloudMeshes)
  (Utilities|FlowControl|Sequence
    (:then_0
      (Rendering|ClearRenderTarget2D (Variables|Default|GetRT_Comp))
      (Rendering|Material|SetScalarParameterValue (Variables|Default|GetMaterial_Comp) "Comp" 0.0 :!class /Script/Engine.MaterialInstanceDynamic)
      (Rendering|DrawMaterialtoRenderTarget (Variables|Default|GetRT_Comp) (Variables|Default|GetMaterial_Comp))
      (EditorScripting|Rendering|ConvertRenderTargettoTexture2DEditorOnly (Variables|Default|GetRT_Comp) (Variables|Output|GetBakedTexture_A))
      (Rendering|ClearRenderTarget2D (Variables|Default|GetRT_Comp))
      (Rendering|Material|SetScalarParameterValue (Variables|Default|GetMaterial_Comp) "Comp" 1.0 :!class /Script/Engine.MaterialInstanceDynamic)
      (Rendering|DrawMaterialtoRenderTarget (Variables|Default|GetRT_Comp) (Variables|Default|GetMaterial_Comp))
      (EditorScripting|Rendering|ConvertRenderTargettoTexture2DEditorOnly (Variables|Default|GetRT_Comp) (Variables|Output|GetBakedTexture_B)))
    (:then_1
      (Variables|/Script/Engine.Texture|SetCompressionSettings "TC_BC7" _bakedtexture_a)
      (Variables|/Script/Engine.Texture|SetSRGB false _bakedtexture_a)
      (Variables|/Script/Engine.Texture|SetCompressionNoAlpha false _bakedtexture_a)
      (Variables|/Script/Engine.Texture|SetDeferCompression true _bakedtexture_a)
      (Variables|/Script/Engine.Texture|SetCompressionSettings "TC_BC7" _bakedtexture_b)
      (Variables|/Script/Engine.Texture|SetSRGB false _bakedtexture_b)
      (Variables|/Script/Engine.Texture|SetCompressionNoAlpha false _bakedtexture_b)
      (Variables|/Script/Engine.Texture|SetDeferCompression true _bakedtexture_b))
    (:then_2
      (Default|DrawPreview))))

(fn DrawPreview
  (Rendering|ClearRenderTarget2D (Variables|Default|GetRT_Comp))
  (Rendering|Material|SetScalarParameterValue (Variables|Default|GetMaterial_Comp) "Comp" (Utilities|Select :Index (Variables|Preview|GetTexturePreview) :NewEnumerator0 0.0 :NewEnumerator1 1.0 :NewEnumerator2 0.0 :NewEnumerator3 0.0 :NewEnumerator4 0.0 :NewEnumerator5 1.0 :NewEnumerator6 1.0 :NewEnumerator7 1.0 :NewEnumerator8 1.0 :NewEnumerator9 0.0) :!class /Script/Engine.MaterialInstanceDynamic)
  (Rendering|DrawMaterialtoRenderTarget (Variables|Default|GetRT_Comp) (Variables|Default|GetMaterial_Comp))
  (bind _returnvalue (Rendering|Material|CreateDynamicMaterialInstance "/Game/NeverCook/EditorUtilities/BakedCloudsPacker/M_BakedCloudsPacker_Preview.M_BakedCloudsPacker_Preview" "None" "None" :!class /Script/Engine.KismetMaterialLibrary))
  (bind _output_get (Variables|Default|SetMaterial_Preview _returnvalue))
  (Rendering|Material|SetTextureParameterValue _output_get "PreviewTexture" (Variables|Default|GetRT_Comp))
  (Rendering|Material|SetVectorParameterValue (Variables|Default|GetMaterial_Preview) "ChannelMask" (Utilities|Select :Index (Variables|Preview|GetTexturePreview) :NewEnumerator0 "(R=0.000000,G=0.000000,B=0.000000,A=0.000000)" :NewEnumerator1 "(R=0.000000,G=0.000000,B=0.000000,A=0.000000)" :NewEnumerator2 "(R=1.000000,G=0.000000,B=0.000000,A=0.000000)" :NewEnumerator3 "(R=0.000000,G=1.000000,B=0.000000,A=0.000000)" :NewEnumerator4 "(R=0.000000,G=0.000000,B=1.000000,A=0.000000)" :NewEnumerator5 "(R=1.000000,G=0.000000,B=0.000000,A=0.000000)" :NewEnumerator6 "(R=0.000000,G=1.000000,B=0.000000,A=0.000000)" :NewEnumerator7 "(R=0.000000,G=0.000000,B=1.000000,A=0.000000)" :NewEnumerator8 "(R=0.000000,G=0.000000,B=0.000000,A=1.000000)" :NewEnumerator9 "(R=0.000000,G=0.000000,B=0.000000,A=1.000000)") :!class /Script/Engine.MaterialInstanceDynamic)
  (Rendering|Material|SetMaterial (Variables|Default|GetPreviewMesh) 0 (Variables|Default|GetMaterial_Preview))
  (if (not (Utilities|Select :Index (Variables|Preview|GetMeshPreview) :NewEnumerator0 true :NewEnumerator10 false))
    (bind _staticmesh (Variables|Output|GetStaticMesh)
      (:then
        (Components|StaticMesh|SetStaticMesh (Variables|Default|GetPreviewMesh) _staticmesh)
        (Transformation|SetRelativeScale3D (Variables|Default|GetPreviewMesh) "1.000000,1.000000,1.000000")
        (Transformation|SetRelativeRotation (Variables|Default|GetPreviewMesh) "0.000000,0.000000,-90.000000" false false))
      (:else
        _)))
  (Components|StaticMesh|SetStaticMesh (Variables|Default|GetPreviewMesh) "/Engine/BasicShapes/Plane.Plane")
  (Transformation|SetRelativeRotation (Variables|Default|GetPreviewMesh) "0, 0, 0" false false)
  (Transformation|SetRelativeScale3D (Variables|Default|GetPreviewMesh) (Math|Vector|MakeVector (/ (Variables|Adjustments|GetMeshSizeX) 1e+02) (/ (Variables|Adjustments|GetMeshSizeY) 1e+02) 5.0)))

(fn ConstructionScript
  (CallFunction|BakeCloudTextures))
