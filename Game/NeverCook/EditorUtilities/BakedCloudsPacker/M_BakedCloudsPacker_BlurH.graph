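# Unlit editor-utility material for the baked clouds packer that computes a horizontal blur of the AlphaTexture parameter's alpha channel.
# It averages 17 equally weighted taps, a center plus eight on each side along U, forming a box filter.
# The Radius parameter sets tap spacing in 1/1024 UV units, assuming a fixed texture width.
# All taps read a fixed lower mip through clamped samplers.
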Material /Game/NeverCook/EditorUtilities/BakedCloudsPacker/M_BakedCloudsPacker_BlurH
graph {
 properties {bAutomaticallySetUsageInEditor=false, shadingModel=MSM_Unlit, usedShadingModels=Unlit}
 nodes [
  n1 = TextureCoordinate()
  n2 = NamedRerouteUsage()
  n3 = Constant() {r=2}
  n4 = TextureCoordinate()
  n5 = ComponentMask(None=@n4) {r=true}
  n6 = Divide(A=NamedRerouteUsage()) {constB=1024}
  n7 = Multiply(A=@n6)
  n8 = TextureSample!(UVs=AppendVector(A=Add(A=@n5, B=@n7), B=ComponentMask(None=@n4) {g=true}), Tex=@n2, Level=@n3, Apply View MipBias) {mipValueMode=TMVM_MipLevel, samplerSource=SSM_Clamp_WorldGroupSettings, texture=/Engine/EngineResources/DefaultTexture}
  n9 = Multiply(A=@n6) {constB=-1}
  n10 = TextureSample(UVs=AppendVector(A=Add(A=@n5, B=@n9), B=ComponentMask(None=@n4) {g=true}), Tex=@n2, Level=@n3) {mipValueMode=TMVM_MipLevel, samplerSource=SSM_Clamp_WorldGroupSettings, texture=&DefaultTexture}
  n11 = Multiply(A=@n6) {constB=2}
  n12 = TextureSample(UVs=AppendVector(A=Add(A=@n5, B=@n11), B=ComponentMask(None=@n4) {g=true}), Tex=@n2, Level=@n3) {mipValueMode=TMVM_MipLevel, samplerSource=SSM_Clamp_WorldGroupSettings, texture=&DefaultTexture}
  n13 = Multiply(A=@n6) {constB=-2}
  n14 = TextureSample(UVs=AppendVector(A=Add(A=@n5, B=@n13), B=ComponentMask(None=@n4) {g=true}), Tex=@n2, Level=@n3) {mipValueMode=TMVM_MipLevel, samplerSource=SSM_Clamp_WorldGroupSettings, texture=&DefaultTexture}
  n15 = Multiply(A=@n6) {constB=3}
  n16 = TextureSample(UVs=AppendVector(A=Add(A=@n5, B=@n15), B=ComponentMask(None=@n4) {g=true}), Tex=@n2, Level=@n3) {mipValueMode=TMVM_MipLevel, samplerSource=SSM_Clamp_WorldGroupSettings, texture=&DefaultTexture}
  n17 = Multiply(A=@n6) {constB=-3}
  n18 = TextureSample(UVs=AppendVector(A=Add(A=@n5, B=@n17), B=ComponentMask(None=@n4) {g=true}), Tex=@n2, Level=@n3) {mipValueMode=TMVM_MipLevel, samplerSource=SSM_Clamp_WorldGroupSettings, texture=&DefaultTexture}
  n19 = Multiply(A=@n6) {constB=4}
  n20 = TextureSample(UVs=AppendVector(A=Add(A=@n5, B=@n19), B=ComponentMask(None=@n4) {g=true}), Tex=@n2, Level=@n3) {mipValueMode=TMVM_MipLevel, samplerSource=SSM_Clamp_WorldGroupSettings, texture=&DefaultTexture}
  n21 = Multiply(A=@n6) {constB=-4}
  n22 = TextureSample(UVs=AppendVector(A=Add(A=@n5, B=@n21), B=ComponentMask(None=@n4) {g=true}), Tex=@n2, Level=@n3) {mipValueMode=TMVM_MipLevel, samplerSource=SSM_Clamp_WorldGroupSettings, texture=&DefaultTexture}
  n23 = Add(A=Add(A=Add(A=@n8.A, B=@n10.A), B=Add(A=@n12.A, B=@n14.A)), B=Add(A=Add(A=@n16.A, B=@n18.A), B=Add(A=@n20.A, B=@n22.A)))
  n24 = TextureCoordinate()
  n25 = ComponentMask(None=@n24) {r=true}
  n26 = Divide(A=NamedRerouteUsage()) {constB=1024}
  n27 = Multiply(A=@n26) {constB=5}
  n28 = NamedRerouteUsage()
  n29 = Constant() {r=2}
  n30 = TextureSample(UVs=AppendVector(A=Add(A=@n25, B=@n27), B=ComponentMask(None=@n24) {g=true}), Tex=@n28, Level=@n29) {mipValueMode=TMVM_MipLevel, samplerSource=SSM_Clamp_WorldGroupSettings, texture=&DefaultTexture}
  n31 = Multiply(A=@n26) {constB=-5}
  n32 = TextureSample(UVs=AppendVector(A=Add(A=@n25, B=@n31), B=ComponentMask(None=@n24) {g=true}), Tex=@n28, Level=@n29) {mipValueMode=TMVM_MipLevel, samplerSource=SSM_Clamp_WorldGroupSettings, texture=&DefaultTexture}
  n33 = Multiply(A=@n26) {constB=6}
  n34 = TextureSample(UVs=AppendVector(A=Add(A=@n25, B=@n33), B=ComponentMask(None=@n24) {g=true}), Tex=@n28, Level=@n29) {mipValueMode=TMVM_MipLevel, samplerSource=SSM_Clamp_WorldGroupSettings, texture=&DefaultTexture}
  n35 = Multiply(A=@n26) {constB=-6}
  n36 = TextureSample(UVs=AppendVector(A=Add(A=@n25, B=@n35), B=ComponentMask(None=@n24) {g=true}), Tex=@n28, Level=@n29) {mipValueMode=TMVM_MipLevel, samplerSource=SSM_Clamp_WorldGroupSettings, texture=&DefaultTexture}
  n37 = Multiply(A=@n26) {constB=7}
  n38 = TextureSample(UVs=AppendVector(A=Add(A=@n25, B=@n37), B=ComponentMask(None=@n24) {g=true}), Tex=@n28, Level=@n29) {mipValueMode=TMVM_MipLevel, samplerSource=SSM_Clamp_WorldGroupSettings, texture=&DefaultTexture}
  n39 = Multiply(A=@n26) {constB=-7}
  n40 = TextureSample(UVs=AppendVector(A=Add(A=@n25, B=@n39), B=ComponentMask(None=@n24) {g=true}), Tex=@n28, Level=@n29) {mipValueMode=TMVM_MipLevel, samplerSource=SSM_Clamp_WorldGroupSettings, texture=&DefaultTexture}
  n41 = Multiply(A=@n26) {constB=8}
  n42 = TextureSample(UVs=AppendVector(A=Add(A=@n25, B=@n41), B=ComponentMask(None=@n24) {g=true}), Tex=@n28, Level=@n29) {mipValueMode=TMVM_MipLevel, samplerSource=SSM_Clamp_WorldGroupSettings, texture=&DefaultTexture}
  n43 = Multiply(A=@n26) {constB=-8}
  n44 = TextureSample(UVs=AppendVector(A=Add(A=@n25, B=@n43), B=ComponentMask(None=@n24) {g=true}), Tex=@n28, Level=@n29) {mipValueMode=TMVM_MipLevel, samplerSource=SSM_Clamp_WorldGroupSettings, texture=&DefaultTexture}
  n45 = Add(A=Add(A=Add(A=@n30.A, B=@n32.A), B=Add(A=@n34.A, B=@n36.A)), B=Add(A=Add(A=@n38.A, B=@n40.A), B=Add(A=@n42.A, B=@n44.A)))
  Divide(A=Add(
    A=TextureSample(UVs=@n1, Tex=@n2, Level=@n3) {mipValueMode=TMVM_MipLevel, samplerSource=SSM_Clamp_WorldGroupSettings, texture=&DefaultTexture}.A
    B=Add(A=@n23, B=@n45)
   )) {constB=17}
  NamedRerouteDeclaration(None=TextureObjectParameter() {parameterName=AlphaTexture}) {name=Texture, nodeColor=<rgba 0 1 0.5781 1>}
  NamedRerouteDeclaration(None=ScalarParameter() {defaultValue=2, parameterName=Radius}) {name=Radius, nodeColor=<rgba 1 0.2578 0 1>}
 ]
}
