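# Unlit editor-utility material for the baked clouds packer that applies a vertical blur to the red channel of the AlphaTexture parameter.
# It averages 17 equally weighted, clamp-sampled taps: the center plus eight on each side along V, forming a box filter.
# The Radius parameter sets the spacing between taps rather than the total extent, and is divided by a hard-coded 1024 to convert it to UV units.
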Material /Game/NeverCook/EditorUtilities/BakedCloudsPacker/M_BakedCloudsPacker_BlurV
graph {
 properties {bAutomaticallySetUsageInEditor=false, shadingModel=MSM_Unlit, usedShadingModels=Unlit}
 nodes [
  n1 = TextureCoordinate()
  n2 = NamedRerouteUsage()
  n3 = TextureCoordinate()
  n4 = ComponentMask(None=@n3) {r=true}
  n5 = ComponentMask(None=@n3) {g=true}
  n6 = Divide(A=NamedRerouteUsage()) {constB=1024}
  n7 = Multiply(A=@n6)
  n8 = TextureSample(UVs=AppendVector(A=@n4, B=Add(A=@n5, B=@n7)), Tex=@n2) {samplerSource=SSM_Clamp_WorldGroupSettings, texture=/Engine/EngineResources/DefaultTexture}
  n9 = Multiply(A=@n6) {constB=-1}
  n10 = TextureSample(UVs=AppendVector(A=@n4, B=Add(A=@n5, B=@n9)), Tex=@n2) {samplerSource=SSM_Clamp_WorldGroupSettings, texture=&DefaultTexture}
  n11 = Multiply(A=@n6) {constB=2}
  n12 = TextureSample(UVs=AppendVector(A=@n4, B=Add(A=@n5, B=@n11)), Tex=@n2) {samplerSource=SSM_Clamp_WorldGroupSettings, texture=&DefaultTexture}
  n13 = Multiply(A=@n6) {constB=-2}
  n14 = TextureSample(UVs=AppendVector(A=@n4, B=Add(A=@n5, B=@n13)), Tex=@n2) {samplerSource=SSM_Clamp_WorldGroupSettings, texture=&DefaultTexture}
  n15 = Multiply(A=@n6) {constB=3}
  n16 = TextureSample(UVs=AppendVector(A=@n4, B=Add(A=@n5, B=@n15)), Tex=@n2) {samplerSource=SSM_Clamp_WorldGroupSettings, texture=&DefaultTexture}
  n17 = Multiply(A=@n6) {constB=-3}
  n18 = TextureSample(UVs=AppendVector(A=@n4, B=Add(A=@n5, B=@n17)), Tex=@n2) {samplerSource=SSM_Clamp_WorldGroupSettings, texture=&DefaultTexture}
  n19 = Multiply(A=@n6) {constB=4}
  n20 = TextureSample(UVs=AppendVector(A=@n4, B=Add(A=@n5, B=@n19)), Tex=@n2) {samplerSource=SSM_Clamp_WorldGroupSettings, texture=&DefaultTexture}
  n21 = Multiply(A=@n6) {constB=-4}
  n22 = TextureSample(UVs=AppendVector(A=@n4, B=Add(A=@n5, B=@n21)), Tex=@n2) {samplerSource=SSM_Clamp_WorldGroupSettings, texture=&DefaultTexture}
  n23 = Add(A=Add(A=Add(A=@n8.R, B=@n10.R), B=Add(A=@n12.R, B=@n14.R)), B=Add(A=Add(A=@n16.R, B=@n18.R), B=Add(A=@n20.R, B=@n22.R)))
  n24 = TextureCoordinate()
  n25 = ComponentMask(None=@n24) {r=true}
  n26 = ComponentMask(None=@n24) {g=true}
  n27 = Divide(A=NamedRerouteUsage()) {constB=1024}
  n28 = Multiply(A=@n27) {constB=5}
  n29 = NamedRerouteUsage()
  n30 = TextureSample(UVs=AppendVector(A=@n25, B=Add(A=@n26, B=@n28)), Tex=@n29) {samplerSource=SSM_Clamp_WorldGroupSettings, texture=&DefaultTexture}
  n31 = Multiply(A=@n27) {constB=-5}
  n32 = TextureSample(UVs=AppendVector(A=@n25, B=Add(A=@n26, B=@n31)), Tex=@n29) {samplerSource=SSM_Clamp_WorldGroupSettings, texture=&DefaultTexture}
  n33 = Multiply(A=@n27) {constB=6}
  n34 = TextureSample(UVs=AppendVector(A=@n25, B=Add(A=@n26, B=@n33)), Tex=@n29) {samplerSource=SSM_Clamp_WorldGroupSettings, texture=&DefaultTexture}
  n35 = Multiply(A=@n27) {constB=-6}
  n36 = TextureSample(UVs=AppendVector(A=@n25, B=Add(A=@n26, B=@n35)), Tex=@n29) {samplerSource=SSM_Clamp_WorldGroupSettings, texture=&DefaultTexture}
  n37 = Multiply(A=@n27) {constB=7}
  n38 = TextureSample(UVs=AppendVector(A=@n25, B=Add(A=@n26, B=@n37)), Tex=@n29) {samplerSource=SSM_Clamp_WorldGroupSettings, texture=&DefaultTexture}
  n39 = Multiply(A=@n27) {constB=-7}
  n40 = TextureSample(UVs=AppendVector(A=@n25, B=Add(A=@n26, B=@n39)), Tex=@n29) {samplerSource=SSM_Clamp_WorldGroupSettings, texture=&DefaultTexture}
  n41 = Multiply(A=@n27) {constB=8}
  n42 = TextureSample(UVs=AppendVector(A=@n25, B=Add(A=@n26, B=@n41)), Tex=@n29) {samplerSource=SSM_Clamp_WorldGroupSettings, texture=&DefaultTexture}
  n43 = Multiply(A=@n27) {constB=-8}
  n44 = TextureSample(UVs=AppendVector(A=@n25, B=Add(A=@n26, B=@n43)), Tex=@n29) {samplerSource=SSM_Clamp_WorldGroupSettings, texture=&DefaultTexture}
  n45 = Add(A=Add(A=Add(A=@n30.R, B=@n32.R), B=Add(A=@n34.R, B=@n36.R)), B=Add(A=Add(A=@n38.R, B=@n40.R), B=Add(A=@n42.R, B=@n44.R)))
  Divide(A=Add(
    A=TextureSample(UVs=@n1, Tex=@n2) {samplerSource=SSM_Clamp_WorldGroupSettings, texture=&DefaultTexture}.R
    B=Add(A=@n23, B=@n45)
   )) {constB=17}
  NamedRerouteDeclaration(None=TextureObjectParameter() {parameterName=AlphaTexture}) {name=Texture, nodeColor=<rgba 0 1 0.5781 1>}
  NamedRerouteDeclaration(None=ScalarParameter() {defaultValue=14.752, parameterName=Radius}) {name=Radius, nodeColor=<rgba 1 0.2578 0 1>}
 ]
}
