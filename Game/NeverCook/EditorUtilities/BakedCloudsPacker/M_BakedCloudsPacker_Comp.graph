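# Unlit alpha-composite material for the BakedCloudsPacker editor utility, which composites baked cloud textures.
# Samples Pack_A, Pack_B and Blur textures, with UV controls that offset and scale around the center.
# A squared-falloff vertical gradient between BlurStart and BlurEnd forms a mask, and computed mip levels soften the pack samples.
# Alphas combine through screen-style blending, and a Comp scalar interpolates between composite inputs.
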Material /Game/NeverCook/EditorUtilities/BakedCloudsPacker/M_BakedCloudsPacker_Comp
graph {
 properties {bAutomaticallySetUsageInEditor=false, blendMode=BLEND_AlphaComposite, shadingModel=MSM_Unlit, usedShadingModels=Unlit}
 nodes [
  n1 = TextureCoordinate()
  n2 = Saturate(None=InverseLinearInterpolate(
    A=ScalarParameter() {defaultValue=0.5, parameterName=BlurEnd}
    B=ScalarParameter() {defaultValue=0.75, parameterName=BlurStart}
    Value=ComponentMask(None=@n1) {g=true}
   ) {constA=0.512, constB=0.8})
  n3 = OneMinus(None=Multiply(A=@n2, B=@n2))
  n4 = Multiply(A=@n3, B=@n3)
  n5 = NamedRerouteUsage()
  n6 = Multiply(A=@n4, B=@n5)
  n7 = NamedRerouteUsage()
  Add(A=@n6, B=@n7)
  LinearInterpolate(A=NamedRerouteUsage(), B=NamedRerouteUsage(), Alpha=ScalarParameter() {parameterName=Comp})
  Max(A=@n6, B=@n7)
  NamedRerouteDeclaration(None=TextureSample(UVs=NamedRerouteUsage(), Tex=TextureObjectParameter() {parameterName=Blur}) {samplerSource=SSM_Clamp_WorldGroupSettings, texture=/Engine/EngineResources/DefaultTexture}.R) {name=Blur, nodeColor=<rgba 0 0.6719 1 1>}
  n8 = NamedRerouteUsage()
  n9 = Add(A=Multiply(A=NamedRerouteUsage()) {constB=2}, B=Multiply(A=OneMinus(None=@n8)) {constB=3})
  n10 = TextureSample!(UVs=NamedRerouteUsage(), Tex=TextureObjectParameter() {parameterName=Pack_B}, Level=@n9, Apply View MipBias) {mipValueMode=TMVM_MipLevel, samplerSource=SSM_Clamp_WorldGroupSettings, texture=&DefaultTexture}
  NamedRerouteDeclaration(None=@n10.RGB) {name=Pack_B_RGB, nodeColor=<rgba 0 0.2031 1 1>}
  NamedRerouteDeclaration(None=@n10.A) {name=Pack_B_Alpha, nodeColor=<rgba 0.5234 1 0 1>}
  NamedRerouteDeclaration(None=OneMinus(None=@n4)) {name=Vertical Mask, nodeColor=<rgba 0 1 0.2031 1>}
  n11 = Subtract(
   A=Add(A=TextureCoordinate()) {constB=-0.5}
   B=AppendVector(A=ScalarParameter() {parameterName=OffsetX}, B=ScalarParameter() {parameterName=OffsetY})
  )
  n12 = AppendVector(A=ScalarParameter() {defaultValue=1, parameterName=ScaleX}, B=ScalarParameter() {defaultValue=1, parameterName=ScaleY})
  NamedRerouteDeclaration(None=Add(A=Divide(A=@n11, B=@n12)) {constB=0.5}) {name=UVs, nodeColor=<rgba 0.8047 0 1 1>}
  n13 = LinearInterpolate(A=@n5, Alpha=NamedRerouteUsage()) {constA=1}
  n14 = OneMinus(None=Multiply(A=OneMinus(None=@n6), B=OneMinus(None=@n7)))
  NamedRerouteDeclaration(None=Multiply(A=LinearInterpolate(Alpha=@n13) {constA=0.75}, B=Saturate(None=@n14))) {name=Comp A Alpha, nodeColor=<rgba 0 0.7656 1 1>}
  n15 = TextureObjectParameter() {parameterName=Pack_A}
  n16 = TextureSample(UVs=NamedRerouteUsage(), Tex=@n15, Level=@n9) {mipValueMode=TMVM_MipLevel, samplerSource=SSM_Clamp_WorldGroupSettings, texture=&DefaultTexture}
  NamedRerouteDeclaration(None=@n16.RGB) {name=Pack_A_RGB, nodeColor=<rgba 0.8281 1 0 1>}
  NamedRerouteDeclaration(None=TextureSample!(UVs=NamedRerouteUsage(), Tex=@n15, Apply View MipBias) {samplerSource=SSM_Clamp_WorldGroupSettings, texture=&DefaultTexture}.A) {name=Pack_A_Alpha, nodeColor=<rgba 0.8607 1 0.9512 1>}
  OneMinus(None=LinearInterpolate(A=NamedRerouteUsage(), B=NamedRerouteUsage(), Alpha=ScalarParameter() {parameterName=Comp}))
  Reroute(None=@n16.RGB)
 ]
}
