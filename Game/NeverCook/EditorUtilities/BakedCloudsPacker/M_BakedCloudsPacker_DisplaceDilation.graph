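# Unlit editor-utility material for the baked clouds packer that builds a dilated binary mask from the BlurredAlpha texture.
# UVs are offset and scaled around the texture center through OffsetX/Y and ScaleX/Y parameters.
# Eight clamped red-channel taps, four axial and four diagonal shortened to approximate a circle, are spaced by DilationAmount in 1024-texel units.
# Their maximum is thresholded with a small subtraction and Ceil.
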Material /Game/NeverCook/EditorUtilities/BakedCloudsPacker/M_BakedCloudsPacker_DisplaceDilation
graph {
 properties {bAutomaticallySetUsageInEditor=false, shadingModel=MSM_Unlit, usedShadingModels=Unlit}
 nodes [
  n1 = NamedRerouteUsage()
  n2 = NamedRerouteUsage()
  n3 = AppendVector(A=Add(A=ComponentMask(None=@n1) {r=true}, B=@n2), B=ComponentMask(None=@n1) {g=true})
  n4 = NamedRerouteUsage()
  n5 = AppendVector(A=Subtract(A=ComponentMask(None=@n1) {r=true}, B=@n2), B=ComponentMask(None=@n1) {g=true})
  n6 = AppendVector(A=ComponentMask(None=@n1) {r=true}, B=Add(A=ComponentMask(None=@n1) {g=true}, B=@n2))
  n7 = AppendVector(A=ComponentMask(None=@n1) {r=true}, B=Subtract(A=ComponentMask(None=@n1) {g=true}, B=@n2))
  n8 = Max(
   A=Max(
    A=TextureSample(UVs=@n3, Tex=@n4) {samplerSource=SSM_Clamp_WorldGroupSettings}.R
    B=TextureSample(UVs=@n5, Tex=@n4) {samplerSource=SSM_Clamp_WorldGroupSettings}.R
   )
   B=Max(
    A=TextureSample(UVs=@n6, Tex=@n4) {samplerSource=SSM_Clamp_WorldGroupSettings}.R
    B=TextureSample(UVs=@n7, Tex=@n4) {samplerSource=SSM_Clamp_WorldGroupSettings}.R
   )
  )
  n9 = NamedRerouteUsage()
  n10 = Divide(A=NamedRerouteUsage()) {constB=1.42}
  n11 = AppendVector(A=Add(A=ComponentMask(None=@n9) {r=true}, B=@n10), B=Add(A=ComponentMask(None=@n9) {g=true}, B=@n10))
  n12 = AppendVector(A=Subtract(A=ComponentMask(None=@n9) {r=true}, B=@n10), B=Subtract(A=ComponentMask(None=@n9) {g=true}, B=@n10))
  n13 = AppendVector(A=Subtract(A=ComponentMask(None=@n9) {r=true}, B=@n10), B=Add(A=ComponentMask(None=@n9) {g=true}, B=@n10))
  n14 = AppendVector(A=Add(A=ComponentMask(None=@n9) {r=true}, B=@n10), B=Subtract(A=ComponentMask(None=@n9) {g=true}, B=@n10))
  n15 = Max(
   A=Max(
    A=TextureSample(UVs=@n11, Tex=@n4) {samplerSource=SSM_Clamp_WorldGroupSettings}.R
    B=TextureSample(UVs=@n12, Tex=@n4) {samplerSource=SSM_Clamp_WorldGroupSettings}.R
   )
   B=Max(
    A=TextureSample(UVs=@n13, Tex=@n4) {samplerSource=SSM_Clamp_WorldGroupSettings}.R
    B=TextureSample(UVs=@n14, Tex=@n4) {samplerSource=SSM_Clamp_WorldGroupSettings}.R
   )
  )
  Ceil(None=Subtract(A=Max(A=@n8, B=@n15)) {constB=0.01})
  NamedRerouteDeclaration(None=Divide(A=ScalarParameter() {parameterName=DilationAmount}) {constA=1024, constB=1024}) {name=Offset, nodeColor=<rgba 0 0.8594 1 1>}
  NamedRerouteDeclaration(None=TextureObjectParameter() {parameterName=BlurredAlpha}) {name=Texture, nodeColor=<rgba 0.9453 0 1 1>}
  n16 = Subtract(
   A=Add(A=TextureCoordinate()) {constB=-0.5}
   B=AppendVector(A=ScalarParameter() {parameterName=OffsetX}, B=ScalarParameter() {parameterName=OffsetY})
  )
  n17 = AppendVector(A=ScalarParameter() {defaultValue=1, parameterName=ScaleX}, B=ScalarParameter() {defaultValue=1, parameterName=ScaleY})
  NamedRerouteDeclaration(None=Add(A=Divide(A=@n16, B=@n17)) {constB=0.5}) {name=UVs, nodeColor=<rgba 0.8047 0 1 1>}
 ]
}
