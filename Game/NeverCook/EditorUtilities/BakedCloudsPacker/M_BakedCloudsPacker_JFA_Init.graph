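# Unlit editor utility material for the baked clouds packer that writes the seed image for a jump flood (JFA) pass.
# It samples the AlphaTexture parameter's alpha and offsets and rounds it into a binary mask with a threshold near 0.75.
# The mask selects each pixel's own UV as a seed or (-1, -1) as an empty marker, relying on negative emissive output.
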Material /Game/NeverCook/EditorUtilities/BakedCloudsPacker/M_BakedCloudsPacker_JFA_Init
graph {
 properties {bAllowNegativeEmissiveColor=true, bAutomaticallySetUsageInEditor=false, shadingModel=MSM_Unlit, usedShadingModels=Unlit}
 nodes [
  n1 = TextureSampleParameter2D() {parameterName=AlphaTexture}
  LinearInterpolate(A=Constant2Vector() {g=-1, r=-1}, B=TextureCoordinate(), Alpha=Round(None=Add(A=@n1.A) {constB=-0.25}))
 ]
}
