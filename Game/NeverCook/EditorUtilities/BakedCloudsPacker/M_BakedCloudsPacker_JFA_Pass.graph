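# Unlit editor-utility material for the baked clouds packer that performs one Jump Flood Algorithm pass over the RT texture.
# A Custom HLSL node reads stored seed coordinates from the 3x3 neighbourhood spaced by the Step parameter, using point-sampled UVs.
# It keeps the non-negative seed nearest each texel, or (-1,-1) when none is found.
# Negative emissive is allowed, letting that empty-seed marker survive.
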Material /Game/NeverCook/EditorUtilities/BakedCloudsPacker/M_BakedCloudsPacker_JFA_Pass
graph {
 properties {bAllowNegativeEmissiveColor=true, bAutomaticallySetUsageInEditor=false, shadingModel=MSM_Unlit, usedShadingModels=Unlit}
 nodes [
  MaterialExpressionTextureObjectParameter_0 = TextureObjectParameter() {parameterName=RT}
  MaterialExpressionMaterialFunctionCall_0 = MaterialFunctionCall!(UVs=TextureCoordinate(), Texture Size=TextureProperty(Tex=@MaterialExpressionTextureObjectParameter_0)) {materialFunction=/Engine/Functions/Engine_MaterialFunctions01/Texturing/PointSampledUVs}
  MaterialExpressionScalarParameter_0 = ScalarParameter() {defaultValue=0.5, parameterName=Step}
  Custom!(
   UV=@MaterialExpressionMaterialFunctionCall_0.Result
   Tex=@MaterialExpressionTextureObjectParameter_0
   StepSize=@MaterialExpressionScalarParameter_0
  ) {
   code="float bestDistance = 99999;\r\nfloat2 bestUV = float2(-1,-1);\r\n\r\nfor(int y = -1; y<= 1; ++y){\r\n\tfor(int x = -1; x<= 1; ++x){\r\n\t\tfloat2 UV_off = UV + float2(x,y) * StepSize;\r\n\t\tfloat2 tempVaule = Texture2DSample(Tex, TexSampler, UV_off).xy;\r\n\t\tfloat dist = length(tempVaule - UV);\r\n\r\n\t\tif((tempVaule.x >= 0)&&&&(tempVaule.y >= 0)&&&&(dist < bestDistance)){\r\n\t\t\tbestDistance = dist;\r\n\t\t\tbestUV = tempVaule;\r\n\t\t}\r\n\t}\r\n}\r\n\r\nreturn float4(bestUV,0,0);"
   description=JFA
   showCode=true
   inputs [
    {
     inputName=UV
     input {expression=&self:MaterialExpressionMaterialFunctionCall_0}
    }
    {
     inputName=Tex
     input {expression=&self:MaterialExpressionTextureObjectParameter_0}
    }
    {
     inputName=StepSize
     input {expression=&self:MaterialExpressionScalarParameter_0}
    }
   ]
  }
 ]
}
