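# Material function for the editor-only BakedCloudsPacker utility that defringes a texture's RGB around transparent edges.
# It takes eight neighbor samples, four axial and four diagonal shortened by 1.42, with DefringeOffset scaled by 1/1024.
# The per-channel maximum of their squared RGB blends toward the original RGB by the minimum neighbor alpha, so edges take the dilated color.
# Texture alpha is output separately, unmodified.
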MaterialFunction /Game/NeverCook/EditorUtilities/BakedCloudsPacker/MF_BakedClouds_Defringe
graph {nodes [
  n1 = NamedRerouteDeclaration(None=FunctionInput(Preview=TextureCoordinate()) {bUsePreviewValueAsDefault=true, inputName=JFA, inputType=FunctionInput_Vector2, sortPriority=2}) {name=UVs, nodeColor=<rgba 0 0.7422 1 1>}
  n2 = ComponentMask(None=@n1) {r=true}
  n3 = NamedRerouteUsage()
  n4 = NamedRerouteUsage()
  n5 = TextureSample!(UVs=AppendVector(A=Add(A=@n2, B=@n3), B=ComponentMask(None=@n1) {g=true}), Tex=@n4, Level=Constant(), Apply View MipBias) {mipValueMode=TMVM_MipLevel, samplerSource=SSM_Clamp_WorldGroupSettings, texture=/Engine/EngineResources/DefaultTexture}
  n6 = ComponentMask(None=@n1) {r=true}
  n7 = TextureSample(UVs=AppendVector(A=Subtract(A=@n6, B=@n3), B=ComponentMask(None=@n1) {g=true}), Tex=@n4, Level=Constant()) {mipValueMode=TMVM_MipLevel, samplerSource=SSM_Clamp_WorldGroupSettings, texture=&DefaultTexture}
  n8 = ComponentMask(None=@n1) {g=true}
  n9 = TextureSample(UVs=AppendVector(A=ComponentMask(None=@n1) {r=true}, B=Add(A=@n8, B=@n3)), Tex=@n4, Level=Constant()) {mipValueMode=TMVM_MipLevel, samplerSource=SSM_Clamp_WorldGroupSettings, texture=&DefaultTexture}
  n10 = ComponentMask(None=@n1) {g=true}
  n11 = TextureSample(UVs=AppendVector(A=ComponentMask(None=@n1) {r=true}, B=Subtract(A=@n10, B=@n3)), Tex=@n4, Level=Constant()) {mipValueMode=TMVM_MipLevel, samplerSource=SSM_Clamp_WorldGroupSettings, texture=&DefaultTexture}
  n12 = Max(
   A=Max(A=Multiply(A=@n5.RGB, B=@n5.RGB), B=Multiply(A=@n7.RGB, B=@n7.RGB))
   B=Max(A=Multiply(A=@n9.RGB, B=@n9.RGB), B=Multiply(A=@n11.RGB, B=@n11.RGB))
  )
  n13 = NamedRerouteUsage()
  n14 = ComponentMask(None=@n13) {r=true}
  n15 = Divide(A=NamedRerouteUsage()) {constB=1.42}
  n16 = ComponentMask(None=@n13) {g=true}
  n17 = TextureSample(UVs=AppendVector(A=Add(A=@n14, B=@n15), B=Add(A=@n16, B=@n15)), Tex=@n4, Level=Constant()) {mipValueMode=TMVM_MipLevel, samplerSource=SSM_Clamp_WorldGroupSettings, texture=&DefaultTexture}
  n18 = ComponentMask(None=@n13) {r=true}
  n19 = ComponentMask(None=@n13) {g=true}
  n20 = TextureSample(UVs=AppendVector(A=Subtract(A=@n18, B=@n15), B=Subtract(A=@n19, B=@n15)), Tex=@n4, Level=Constant()) {mipValueMode=TMVM_MipLevel, samplerSource=SSM_Clamp_WorldGroupSettings, texture=&DefaultTexture}
  n21 = ComponentMask(None=@n13) {r=true}
  n22 = ComponentMask(None=@n13) {g=true}
  n23 = TextureSample(UVs=AppendVector(A=Subtract(A=@n21, B=@n15), B=Add(A=@n22, B=@n15)), Tex=@n4, Level=Constant()) {mipValueMode=TMVM_MipLevel, samplerSource=SSM_Clamp_WorldGroupSettings, texture=&DefaultTexture}
  n24 = ComponentMask(None=@n13) {r=true}
  n25 = ComponentMask(None=@n13) {g=true}
  n26 = TextureSample(UVs=AppendVector(A=Add(A=@n24, B=@n15), B=Subtract(A=@n25, B=@n15)), Tex=@n4, Level=Constant()) {mipValueMode=TMVM_MipLevel, samplerSource=SSM_Clamp_WorldGroupSettings, texture=&DefaultTexture}
  n27 = Max(
   A=Max(A=Multiply(A=@n17.RGB, B=@n17.RGB), B=Multiply(A=@n20.RGB, B=@n20.RGB))
   B=Max(A=Multiply(A=@n23.RGB, B=@n23.RGB), B=Multiply(A=@n26.RGB, B=@n26.RGB))
  )
  n28 = NamedRerouteUsage()
  n29 = NamedRerouteUsage()
  n30 = Min(A=Min(A=@n5.A, B=@n7.A), B=Min(A=@n9.A, B=@n11.A))
  n31 = Min(A=Min(A=@n17.A, B=@n20.A), B=Min(A=@n23.A, B=@n26.A))
  FunctionOutput(None=LinearInterpolate(
    A=Max(A=@n12, B=@n27)
    B=TextureSample!(UVs=@n28, Tex=@n29, Apply View MipBias) {samplerSource=SSM_Clamp_WorldGroupSettings, texture=&DefaultTexture}.RGB
    Alpha=Min(A=@n30, B=@n31)
   )) {outputName=Texture RGB}
  FunctionOutput(None=TextureSample(Tex=NamedRerouteUsage()) {samplerSource=SSM_Clamp_WorldGroupSettings, texture=&DefaultTexture}.A) {outputName=Texture Alpha, sortPriority=2}
  n32 = FunctionInput() {bUsePreviewValueAsDefault=true, inputName=DefringeOffset, inputType=FunctionInput_Scalar, sortPriority=1, previewValue=<5 5 5 5>}
  NamedRerouteDeclaration(None=Divide(A=@n32) {constA=1024, constB=1024}) {name=Offset, nodeColor=<rgba 0 0.8594 1 1>}
  NamedRerouteDeclaration(None=FunctionInput(Preview=TextureObject()) {inputName=Texture, inputType=FunctionInput_Texture2D}) {name=Texture, nodeColor=<rgba 1 0.7969 0 1>}
 ]}
