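# Unlit alpha-composite editor utility material that packs baked cloud renders FrontLeft, Front and FrontRight into a three-channel color, plus inverted FrontLeft alpha.
# Each input passes through a defringe function driven by a shared radius and a point-sampled JFA texture's RG channels.
# The defringed color reduces to luminance, then per-view Shadows exponents and Highlights multipliers shape it before saturation and channel packing.
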Material /Game/NeverCook/EditorUtilities/BakedCloudsPacker/M_BakedCloudsPacker_Pack_A
graph {
 properties {bAutomaticallySetUsageInEditor=false, blendMode=BLEND_AlphaComposite, shadingModel=MSM_Unlit, usedShadingModels=Unlit}
 nodes [
  n1 = Constant3Vector() {constant=<rgba 0.2122 0.7157 0.0723 1>}
  n2 = ScalarParameter() {parameterName=DefringeRadius}
  n3 = MaterialFunctionCall!(Texture=TextureObjectParameter() {parameterName=FrontLeft}, DefringeOffset=@n2, JFA=NamedRerouteUsage()) {materialFunction=&/MF_BakedClouds_Defringe}
  n4 = Multiply(
   A=Power(Base=DotProduct(A=@n1, B=@n3.Texture RGB), Exp=ScalarParameter() {defaultValue=3, parameterName=FrontLeftShadows}) {constExponent=3}
   B=ScalarParameter() {defaultValue=1.2, parameterName=FrontLeftHighlights}
  ) {constB=1.2}
  n5 = MaterialFunctionCall(Texture=TextureObjectParameter() {parameterName=Front}, DefringeOffset=@n2, JFA=NamedRerouteUsage()) {materialFunction=&MF_BakedClouds_Defringe}
  n6 = Multiply(
   A=Power(Base=DotProduct(A=@n1, B=@n5.Texture RGB), Exp=ScalarParameter() {defaultValue=3, parameterName=FrontShadows}) {constExponent=3}
   B=ScalarParameter() {defaultValue=1.2, parameterName=FrontHighlights}
  ) {constB=1.2}
  n7 = TextureObjectParameter() {parameterName=FrontRight}
  n8 = NamedRerouteUsage()
  n9 = Power(
   Base=DotProduct(
    A=@n1
    B=MaterialFunctionCall(Texture=@n7, DefringeOffset=@n2, JFA=@n8) {materialFunction=&MF_BakedClouds_Defringe}.Texture RGB
   )
   Exp=ScalarParameter() {defaultValue=3, parameterName=FrontRightShadows}
  ) {constExponent=3}
  n10 = ScalarParameter() {defaultValue=1.2, parameterName=FrontRightHighlights}
  AppendVector(A=AppendVector(A=Saturate(None=@n4), B=Saturate(None=@n6)), B=Saturate(None=Multiply(A=@n9, B=@n10) {constB=1.2}))
  n11 = TextureObjectParameter() {parameterName=JFA}
  n12 = MaterialFunctionCall!(UVs=TextureCoordinate(), Texture Size=TextureProperty(Tex=@n11)) {materialFunction=/Engine/Functions/Engine_MaterialFunctions01/Texturing/PointSampledUVs}
  NamedRerouteDeclaration(None=ComponentMask(None=TextureSample(UVs=@n12.Result, Tex=@n11) {texture=/Engine/EngineResources/DefaultTexture}.RGB) {g=true, r=true}) {name=JFA, nodeColor=<rgba 0.6172 0 1 1>}
  OneMinus(None=@n3.Texture Alpha)
 ]
}
