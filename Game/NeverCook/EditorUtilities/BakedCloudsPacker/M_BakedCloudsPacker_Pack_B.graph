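# Unlit editor-utility material that packs four baked cloud textures into channels for the BakedCloudsPacker tool.
# A point-sampled JFA (likely jump flood) texture and a radius parameter feed a defringe function for each texture.
# Each result becomes luminance, shaped by per-texture shadow power and highlight multiplier, then clamped.
# BackLeft, Back and BackRight form RGB; Top is inverted, apparently for alpha through AlphaComposite blending.
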Material /Game/NeverCook/EditorUtilities/BakedCloudsPacker/M_BakedCloudsPacker_Pack_B
graph {
 properties {bAutomaticallySetUsageInEditor=false, blendMode=BLEND_AlphaComposite, shadingModel=MSM_Unlit, usedShadingModels=Unlit}
 nodes [
  n1 = Constant3Vector() {constant=<rgba 0.2122 0.7157 0.0723 1>}
  n2 = ScalarParameter() {parameterName=DefringeRadius}
  n3 = MaterialFunctionCall!(Texture=TextureObjectParameter() {parameterName=BackLeft}, DefringeOffset=@n2, JFA=NamedRerouteUsage()) {materialFunction=&/MF_BakedClouds_Defringe}
  n4 = Multiply(
   A=Power(Base=DotProduct(A=@n1, B=@n3.Texture RGB), Exp=ScalarParameter() {defaultValue=3, parameterName=BackLeftShadows}) {constExponent=3}
   B=ScalarParameter() {defaultValue=1.2, parameterName=BackLeftHighlights}
  ) {constB=1.2}
  n5 = MaterialFunctionCall(Texture=TextureObjectParameter() {parameterName=Back}, DefringeOffset=@n2, JFA=NamedRerouteUsage()) {materialFunction=&MF_BakedClouds_Defringe}
  n6 = Multiply(
   A=Power(Base=DotProduct(A=@n1, B=@n5.Texture RGB), Exp=ScalarParameter() {defaultValue=3, parameterName=BackShadows}) {constExponent=3}
   B=ScalarParameter() {defaultValue=1.2, parameterName=BackHighlights}
  ) {constB=1.2}
  n7 = TextureObjectParameter() {parameterName=BackRight}
  n8 = NamedRerouteUsage()
  n9 = Power(
   Base=DotProduct(
    A=@n1
    B=MaterialFunctionCall(Texture=@n7, DefringeOffset=@n2, JFA=@n8) {materialFunction=&MF_BakedClouds_Defringe}.Texture RGB
   )
   Exp=ScalarParameter() {defaultValue=3, parameterName=BackRightShadows}
  ) {constExponent=3}
  n10 = ScalarParameter() {defaultValue=1.2, parameterName=BackRightHighlights}
  AppendVector(A=AppendVector(A=Saturate(None=@n4), B=Saturate(None=@n6)), B=Saturate(None=Multiply(A=@n9, B=@n10) {constB=1.2}))
  n11 = TextureObjectParameter() {parameterName=JFA}
  n12 = MaterialFunctionCall!(UVs=TextureCoordinate(), Texture Size=TextureProperty(Tex=@n11)) {materialFunction=/Engine/Functions/Engine_MaterialFunctions01/Texturing/PointSampledUVs}
  NamedRerouteDeclaration(None=ComponentMask(None=TextureSample(UVs=@n12.Result, Tex=@n11) {samplerType=SAMPLERTYPE_LinearColor}.RGB) {g=true, r=true}) {name=JFA, nodeColor=<rgba 0.6172 0 1 1>}
  n13 = TextureObjectParameter() {parameterName=Top}
  n14 = NamedRerouteUsage()
  n15 = Power(
   Base=DotProduct(
    A=@n1
    B=MaterialFunctionCall(Texture=@n13, DefringeOffset=@n2, JFA=@n14) {materialFunction=&MF_BakedClouds_Defringe}.Texture RGB
   )
   Exp=ScalarParameter() {defaultValue=3, parameterName=TopShadows}
  ) {constExponent=3}
  n16 = ScalarParameter() {defaultValue=1.2, parameterName=TopHighlights}
  OneMinus(None=Saturate(None=Multiply(A=@n15, B=@n16) {constB=1.2}))
 ]
}
