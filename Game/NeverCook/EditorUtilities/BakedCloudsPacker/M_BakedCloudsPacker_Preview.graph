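# Unlit preview material for the BakedCloudsPacker editor utility that displays a texture parameter, either as RGB or as an isolated channel.
# A dot product of the RGBA sample with the ChannelMask parameter extracts the channel as grayscale.
# The sum of the mask's components drives the blend, so an all-zero mask shows RGB and a one-hot mask shows that channel.
# EyeAdaptationInverse at full strength cancels auto-exposure on the result.
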Material /Game/NeverCook/EditorUtilities/BakedCloudsPacker/M_BakedCloudsPacker_Preview
graph {
 properties {bAutomaticallySetUsageInEditor=false, shadingModel=MSM_Unlit, usedShadingModels=Unlit}
 nodes [
  n1 = TextureSampleParameter2D() {parameterName=PreviewTexture, samplerSource=SSM_Clamp_WorldGroupSettings}
  n2 = VectorParameter() {parameterName=ChannelMask}
  n3 = Constant() {r=1}
  EyeAdaptationInverse(
   LightValueInput=LinearInterpolate(A=@n1.RGB, B=DotProduct(A=@n1.RGBA, B=@n2.RGBA), Alpha=DotProduct(A=@n2.RGBA, B=@n3))
   AlphaInput=Constant() {r=1}
  )
 ]
}
